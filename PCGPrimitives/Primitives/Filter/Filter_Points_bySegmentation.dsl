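; PCG primitive graph that filters points by segmentation on normal, local Z position and optionally color, with an element-count operator and threshold.
; It tags the input "Input:0", then chains input management, core process and output preparation subgraphs.
; The core subgraph receives the toggles, tolerances and count filter; its logic lives outside this graph.
; Soft-object graph parameters make input fallback, input processing, core and output stages replaceable.
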(graph Filter_Points_bySegmentation
  (params (InputsProcessingGraph :type (SoftObject /Script/PCG.PCGGraphInterface) :default /PCGPrimitives/Primitives/_Template_/Primitives_InputsProcessing.Primitives_InputsProcessing) (InputsFallbackGraph :type (SoftObject /Script/PCG.PCGGraphInterface) :default /PCGPrimitives/Primitives/_Template_/Primitives_InputsFallback.Primitives_InputsFallback) (CoreProcessGraph :type (SoftObject /Script/PCG.PCGGraphInterface) :default /PCGPrimitives/Primitives/Filter/Subgraphs/Filter_Points_bySegmentation_CoreProcess.Filter_Points_bySegmentation_CoreProcess) (PrepareOutputsGraph :type (SoftObject /Script/PCG.PCGGraphInterface) :default /PCGPrimitives/Primitives/_Template_/Primitives_PrepareOutputs.Primitives_PrepareOutputs) (UseNormalSegmentation :type Bool :default true) (NormalSegmentationTolerance :type Double :default 0.25) (UseLocalZPositionSegmentation :type Bool :default true) (LocalZPositionSegementationTolerance :type Double :default 1e+01) (UseColorSegmentation :type Bool) (ColorSegmentationTolerance :type Double :default 0.5) (ElementCountFilterOperator :type (Enum /Script/PCG.EPCGAttributeFilterOperator)) (ElementCountFilterThreshold :type Int))
  (inputs (PrimaryInput :type Any))
  (bind AddTags_1 (Add%20Tags :TagsToAdd "Input:0"))
  ;[0 0 0 0.5 32] Inputs
  (bind Subgraph_0 (Subgraph :Graph /PCGPrimitives/Primitives/_Template_/Subgraphs/Primitives_InputsManagement.Primitives_InputsManagement :InputsFallbackGraph InputsFallbackGraph :InputsProcessingGraph InputsProcessingGraph))
  ;[0 0 0 0.5 32] Core Process
  (bind Subgraph_3 (Subgraph :Graph /PCGPrimitives/Primitives/Filter/Subgraphs/Filter_Points_bySegmentation_CoreProcess.Filter_Points_bySegmentation_CoreProcess :ColorSegmentationTolerance ColorSegmentationTolerance :ElementCountFilterOperator ElementCountFilterOperator :ElementCountFilterThreshold ElementCountFilterThreshold :LocalZPositionSegmentationTolerance LocalZPositionSegementationTolerance :NormalSegmentationTolerance NormalSegmentationTolerance :SubgraphOverride-param CoreProcessGraph :UseColorSegementation UseColorSegmentation :UseLocalZPositionSegmentation UseLocalZPositionSegmentation :UseNormalSegmentation UseNormalSegmentation))
  ;[0 0 0 0.5 32] Prepare Output
  (bind Subgraph_4 (Subgraph :Graph /PCGPrimitives/Primitives/_Template_/Primitives_PrepareOutputs.Primitives_PrepareOutputs :SubgraphOverride-param PrepareOutputsGraph))
  (connect DefaultInputNode PrimaryInput AddTags_1 In)
  (connect AddTags_1 Out Subgraph_0 In)
  (connect Subgraph_0 Out Subgraph_3 In)
  (connect Subgraph_3 Out Subgraph_4 In)
  (output Out Subgraph_4))
