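; PCG primitive graph that randomly splits its input into Selected and Unselected outputs, by ratio or count, with a seed and optional per-dataset randomization.
; Input is tagged Input:0, then runs through shared template subgraphs for input management, a random-choice core process and output preparation.
; A final tag filter routes data tagged Selected to one pin and the rest to the other.
; Graph parameters expose the input, core and output stages as overridable subgraphs.
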(graph Filter_RandomChoice
  (params (InputsProcessingGraph :type (SoftObject /Script/PCG.PCGGraphInterface) :default /PCGPrimitives/Primitives/_Template_/Primitives_InputsProcessing.Primitives_InputsProcessing) (InputsFallbackGraph :type (SoftObject /Script/PCG.PCGGraphInterface) :default /PCGPrimitives/Primitives/_Template_/Primitives_InputsFallback.Primitives_InputsFallback) (CoreProcessGraph :type (SoftObject /Script/PCG.PCGGraphInterface) :default /PCGPrimitives/Primitives/Filter/Subgraphs/Filter_RandomChoice_CoreProcess.Filter_RandomChoice_CoreProcess) (PrepareOutputsGraph :type (SoftObject /Script/PCG.PCGGraphInterface) :default /PCGPrimitives/Primitives/_Template_/Primitives_PrepareOutputs.Primitives_PrepareOutputs) (Ratio :type Double :default 0.5) (UseCountInsteadOfRatio :type Bool) (CountToKeep :type Int :default 1) (Seed :type Int) (RandomPerDataset :type Bool :default true))
  (inputs (PrimaryInput :type Any))
  (outputs (Selected :type Any) (Unselected :type Any))
  (bind AddTags_0 (Add%20Tags :TagsToAdd "Input:0"))
  (bind FilterDataByTag_0 (Filter%20Data%20By%20Tag :SelectedTags "Selected"))
  ;[0 0 0 0.5 32] Inputs
  (bind Subgraph_0 (Subgraph :Graph /PCGPrimitives/Primitives/_Template_/Subgraphs/Primitives_InputsManagement.Primitives_InputsManagement :InputsFallbackGraph InputsFallbackGraph :InputsProcessingGraph InputsProcessingGraph))
  ;[0 0 0 0.5 32] Core Process
  (bind Subgraph_3 (Subgraph :Graph /PCGPrimitives/Primitives/Filter/Subgraphs/Filter_RandomChoice_CoreProcess.Filter_RandomChoice_CoreProcess :Count CountToKeep :RandomPerDataset RandomPerDataset :Ratio Ratio :Seed-param Seed :SubgraphOverride-param CoreProcessGraph :Use_Count_instead_of_ratio UseCountInsteadOfRatio))
  ;[0 0 0 0.5 32] Prepare Output
  (bind Subgraph_4 (Subgraph :Graph /PCGPrimitives/Primitives/_Template_/Primitives_PrepareOutputs.Primitives_PrepareOutputs :SubgraphOverride-param PrepareOutputsGraph))
  (connect DefaultInputNode PrimaryInput AddTags_0 In)
  (connect Subgraph_4 Out FilterDataByTag_0 In)
  (connect AddTags_0 Out Subgraph_0 In)
  (connect Subgraph_0 Out Subgraph_3 In)
  (connect Subgraph_3 Out Subgraph_4 In)
  (output Selected FilterDataByTag_0.InsideFilter)
  (output Unselected FilterDataByTag_0.OutsideFilter))
