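; PCG filter primitive that filters input data by relative position, with options for bidirectional checks, direction vector, local point rotation, tolerance threshold and per-element filtering.
; Primary and secondary inputs are tagged by index and gathered into one stream, which runs through input management, core process and output preparation subgraphs.
; Graph parameters make the input fallback, input processing, core and output stages swappable, with core and output using subgraph overrides.
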(graph Filter_RelativePosition
  (params (InputsProcessingGraph :type (SoftObject /Script/PCG.PCGGraphInterface) :default /PCGPrimitives/Primitives/_Template_/Primitives_InputsProcessing.Primitives_InputsProcessing) (InputsFallbackGraph :type (SoftObject /Script/PCG.PCGGraphInterface) :default /PCGPrimitives/Primitives/_Template_/Primitives_InputsFallback.Primitives_InputsFallback) (CoreProcessGraph :type (SoftObject /Script/PCG.PCGGraphInterface) :default /PCGPrimitives/Primitives/Filter/Subgraphs/Filter_RelativePosition_CoreProcess.Filter_RelativePosition_CoreProcess) (PrepareOutputsGraph :type (SoftObject /Script/PCG.PCGGraphInterface) :default /PCGPrimitives/Primitives/Filter/Subgraphs/Filter_RelativePosition_PrepareOutputs.Filter_RelativePosition_PrepareOutputs) (RelativePosition :type (Enum /PCGPrimitives/Primitives/Filter/Resources/EPCG_RelativePositionEnum.EPCG_RelativePositionEnum)) (Bidirectional :type Bool) (UseDirectionVector :type Bool) (Direction :type Vector :default (1.0 0.0 0.0)) (UseLocalPointRotation :type Bool :default true) (UseToleranceThreshold :type Bool) (ToleranceThreshold :type Double) (FilterPerElement :type Bool :default true))
  (inputs (PrimaryInput :type Any) (SecondaryInput :type Any))
  (bind AddTags_0 (Add%20Tags :TagsToAdd "Input:0"))
  (bind AddTags_1 (Add%20Tags :TagsToAdd "Input:1"))
  (bind Gather_1 (Gather))
  ;[0 0 0 0.5 32] Inputs
  (bind Subgraph_0 (Subgraph :Graph /PCGPrimitives/Primitives/_Template_/Subgraphs/Primitives_InputsManagement.Primitives_InputsManagement :InputsFallbackGraph InputsFallbackGraph :InputsProcessingGraph InputsProcessingGraph))
  ;[0 0 0 0.5 32] Core Process
  (bind Subgraph_3 (Subgraph :Graph /PCGPrimitives/Primitives/Filter/Subgraphs/Filter_RelativePosition_CoreProcess.Filter_RelativePosition_CoreProcess :Bidirectional Bidirectional :Direction Direction :FilterPerElement FilterPerElement :RelativePosition RelativePosition :SubgraphOverride-param CoreProcessGraph :ToleranceThreshold ToleranceThreshold :UseDirectionVector UseDirectionVector :UseLocalPointRotation UseLocalPointRotation :UseToleranceThreshold UseToleranceThreshold))
  ;[0 0 0 0.5 32] Prepare Output
  (bind Subgraph_4 (Subgraph :Graph /PCGPrimitives/Primitives/Filter/Subgraphs/Filter_RelativePosition_PrepareOutputs.Filter_RelativePosition_PrepareOutputs :SubgraphOverride-param PrepareOutputsGraph))
  (connect DefaultInputNode PrimaryInput AddTags_0 In)
  (connect DefaultInputNode SecondaryInput AddTags_1 In)
  (connect AddTags_0 Out Gather_1 In)
  (connect AddTags_1 Out Gather_1 In2)
  (connect Gather_1 Out Subgraph_0 In)
  (connect Subgraph_0 Out Subgraph_3 In)
  (connect Subgraph_3 Out Subgraph_4 In)
  (output Out Subgraph_4))
