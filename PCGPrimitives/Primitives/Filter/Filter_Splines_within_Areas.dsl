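; PCG filter primitive that, per its name, sorts splines by containment within areas, with a toggle for partial containment.
; Primary and secondary inputs get index tags, merge, then pass through inputs-management, core-process and prepare-outputs subgraphs.
; Graph parameters make input processing, fallback, core logic and output preparation swappable.
; A final tag filter routes data tagged Selected to the Selected output and the rest to Unselected.
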(graph Filter_Splines_within_Areas
  (params (InputsProcessingGraph :type (SoftObject /Script/PCG.PCGGraphInterface) :default /PCGPrimitives/Primitives/Compose/Subgraphs/Compose_Shared_InputsProcessing.Compose_Shared_InputsProcessing) (InputsFallbackGraph :type (SoftObject /Script/PCG.PCGGraphInterface) :default /PCGPrimitives/Primitives/_Template_/Primitives_InputsFallback.Primitives_InputsFallback) (CoreProcessGraph :type (SoftObject /Script/PCG.PCGGraphInterface) :default /PCGPrimitives/Primitives/Filter/Subgraphs/Filter_Splines_Within_Areas_CoreProcess1.Filter_Splines_Within_Areas_CoreProcess1) (PrepareOutputsGraph :type (SoftObject /Script/PCG.PCGGraphInterface)) (includePartiallyContainedAreas :type Bool))
  (inputs (PrimaryInput :type Any) (SecondaryInput :type Any))
  (outputs (Selected :type Any) (Unselected :type Any))
  (bind AddTags_2 (Add%20Tags :TagsToAdd "Input:0"))
  (bind AddTags_3 (Add%20Tags :TagsToAdd "Input:1"))
  (bind Gather_1 (Gather))
  (bind FilterDataByTag_1 (Filter%20Data%20By%20Tag :SelectedTags "Selected"))
  ;[0 0 0 0.5 32] Inputs
  (bind Subgraph_0 (Subgraph :Graph /PCGPrimitives/Primitives/_Template_/Subgraphs/Primitives_InputsManagement.Primitives_InputsManagement :InputsFallbackGraph InputsFallbackGraph :InputsProcessingGraph InputsProcessingGraph))
  ;[0 0 0 0.5 32] Core Process
  (bind Subgraph_3 (Subgraph :Graph /PCGPrimitives/Primitives/Filter/Subgraphs/Filter_Splines_Within_Areas_CoreProcess1.Filter_Splines_Within_Areas_CoreProcess1 :include_partially_contained_areas includePartiallyContainedAreas :SubgraphOverride-param CoreProcessGraph))
  ;[0 0 0 0.5 32] Prepare Output
  (bind Subgraph_5 (Subgraph :Graph /PCGPrimitives/Primitives/_Template_/Primitives_PrepareOutputs.Primitives_PrepareOutputs :SubgraphOverride-param PrepareOutputsGraph))
  (connect DefaultInputNode PrimaryInput AddTags_2 In)
  (connect DefaultInputNode SecondaryInput AddTags_3 In)
  (connect Subgraph_5 Out FilterDataByTag_1 In)
  (connect AddTags_2 Out Gather_1 In)
  (connect AddTags_3 Out Gather_1 In2)
  (connect Gather_1 Out Subgraph_0 In)
  (connect Subgraph_0 Out Subgraph_3 In)
  (connect Subgraph_3 Out Subgraph_5 In)
  (output Selected FilterDataByTag_1.InsideFilter)
  (output Unselected FilterDataByTag_1.OutsideFilter))
